AnimBlueprint /BRCosmetics/Accessories/FORT_Backpacks/Backpack_M_MED_AutumnFern/Meshes/M_MED_AutumnFern_Pack_AnimBP
target_skeleton=&/M_MED_AutumnFern_Pack_Skeleton
graphs [
 AnimGraph {nodes [Output Pose(Result=Input Pose()) {layerGroup=None}]}
 EventGraph {nodes [
   n1 = Timer Exec() {is=title K2Node_CustomEvent}
   > n2 = Increment Int(Value=Get ShowVar() -> (ShowVar:int=0))
   > n3 = Set ShowVar(ShowVar:int=@n2) -> (Output_Get:int=0)
   > n4 = Switch on Int!(execute:Exec, Selection:int=@n3) -> (Default:Exec, 0:Exec, 1:Exec, 2:Exec, 3:Exec, 4:Exec, 5:Exec)
   n5 = Get MatInstance() -> (MatInstance:Material Instance Dynamic Object Reference)
   SetScalarParameterValue(execute=@n4.1, self=@n5, ParameterName=Hide Element 05, Value=0)
   SetScalarParameterValue(execute=@n4.2, self=@n5, ParameterName=Hide Element 04, Value=0)
   SetScalarParameterValue(execute=@n4.3, self=@n5, ParameterName=Hide Element 03, Value=0)
   SetScalarParameterValue(execute=@n4.4, self=@n5, ParameterName=Hide Element 02, Value=0)
   SetScalarParameterValue(execute=@n4.5, self=@n5, ParameterName=Hide Element 01, Value=0)
   Event OnCharacterCustomizationLoaded()
   > n6 = Get Dynamic Material Instance(SkeletalMeshComponent=GetPartSkeletalMeshComponent())
   > Set MatInstance(MatInstance:Material Instance Dynamic Object Reference=@n6) -> (Output_Get:Material Instance Dynamic Object Reference)
   n7 = Get MatInstance()
   > SetScalarParameterValue(self=@n7, ParameterName=Hide Element 01, Value=1)
   > SetScalarParameterValue(self=@n7, ParameterName=Hide Element 02, Value=1)
   > SetScalarParameterValue(self=@n7, ParameterName=Hide Element 03, Value=1)
   > SetScalarParameterValue(self=@n7, ParameterName=Hide Element 04, Value=1)
   > SetScalarParameterValue(self=@n7, ParameterName=Hide Element 05, Value=1)
   Event BlueprintUpdateAnimation()
   > n8 = Branch(Condition=Get bShouldDisplayFrontendReactivity() -> (bShouldDisplayFrontendReactivity:bool))
   > n9 = Sequence()
   Branch(execute=@n9.then_1, Condition=Get bIsInFrontEnd())
   > Do Once(Reset=@n13.then)
   > Sequence()
   > Set ShowVar(ShowVar=0) -> (Output_Get=0)
   > n10 = Set Timer by Event(Delegate=@n1, Time=float * float(A=Get Timer() -> (Timer:double), B=1.33) {is=title K2Node_PromotableOperator}, bLooping=true)
   > n11 = Set Timer Handle(Timer Handle:Timer Handle Structure=@n10) -> (Output_Get:Timer Handle Structure)
   > n12 = Unpause Timer by Handle!(execute, Handle:Timer Handle Structure=@n11) -> (then)
   Do Once(execute=@n14.then, Reset=@n12.then)
   > Delay(Duration=4)
   > SetScalarParameterValue(self=@n5, ParameterName=Hide Element 01, Value=1)
   > SetScalarParameterValue(self=@n5, ParameterName=Hide Element 02, Value=1)
   > SetScalarParameterValue(self=@n5, ParameterName=Hide Element 03, Value=1)
   > SetScalarParameterValue(self=@n5, ParameterName=Hide Element 04, Value=1)
   > n13 = SetScalarParameterValue(self=@n5, ParameterName=Hide Element 05, Value=1)
   Branch(execute=@n9.then_0, Condition=Get bIsInFrontEnd())
   > n14 = Branch(Condition="integer >= integer"(A=Get ShowVar() -> (ShowVar=0), B=5))
   n15 = Sequence!(execute=@n8.else) -> (then_0, then_1, then_2:Exec, then_3:Exec, then_4:Exec, then_5:Exec)
   > n16 = Cast To FortPlayerPawn(Object=TryGetFortPawn())
   > n17 = WatchCosmeticStat!(
    execute:Exec
    Pawn:Fort Player Pawn Object Reference=@n16
    StatTag:Fort Stat Manager Tag Structure=(TagName="StatManager.AthenaKills")
    DefaultValue:int=0
    bBroadcastInitialValue:bool=false
    InitialValueBroadcastDelay:float=-1
    ChangeValueBroadcastDelay:float=-1
    bConsumeWaitingNotifications:bool=true
   ) -> (then:Exec, OnStatChanged:Exec, StatName:Name, StatValue:int) {is=title K2Node_AsyncAction}
   Set KillCount(execute=@n17.OnStatChanged, KillCount:double="To Float (Integer)"(InInt=@n17.StatValue)) -> (Output_Get:double)
   n18 = Branch(execute=@n15.then_1, Condition="float >= float"(A=Get KillCount() -> (KillCount:double), B=5))
   n19 = Get MatInstance()
   > SetScalarParameterValue(self=@n19, ParameterName=Hide Element 01, Value=0)
   SetScalarParameterValue(execute=@n18.else, self=@n19, ParameterName=Hide Element 01, Value=1)
   n20 = Branch(execute=@n15.then_2, Condition="float >= float"(A=Get KillCount(), B=4))
   > SetScalarParameterValue(self=@n19, ParameterName=Hide Element 02, Value=0)
   SetScalarParameterValue(execute=@n20.else, self=@n19, ParameterName=Hide Element 02, Value=1)
   n21 = Branch(execute=@n15.then_3, Condition="float >= float"(A=Get KillCount(), B=3))
   > SetScalarParameterValue(self=@n19, ParameterName=Hide Element 03, Value=0)
   SetScalarParameterValue(execute=@n21.else, self=@n19, ParameterName=Hide Element 03, Value=1)
   n22 = Branch(execute=@n15.then_4, Condition="float >= float"(A=Get KillCount(), B=2))
   > SetScalarParameterValue(self=@n19, ParameterName=Hide Element 04, Value=0)
   SetScalarParameterValue(execute=@n22.else, self=@n19, ParameterName=Hide Element 04, Value=1)
   n23 = Branch(execute=@n15.then_5, Condition="float >= float"(A=Get KillCount(), B=1))
   > SetScalarParameterValue(self=@n19, ParameterName=Hide Element 05, Value=0)
   SetScalarParameterValue(execute=@n23.else, self=@n19, ParameterName=Hide Element 05, Value=1)
   TryGetPawnOwner()
  ]}
 Update boxes visibility in Frontend {nodes [
   Update boxes visibility in Frontend() {is=title K2Node_FunctionEntry}
   > n1 = Switch on Int!(execute, Selection=Get ShowVar() -> (ShowVar=0)) -> (Default, 0, 1, 2)
   n2 = Get MatInstance()
   SetScalarParameterValue(execute=@n1.1, self=@n2, ParameterName=Hide Element 02, Value=0)
   SetScalarParameterValue(execute=@n1.2, self=@n2, ParameterName=Hide Element 03, Value=0)
   > SetScalarParameterValue(self=@n2, ParameterName=Hide Element 04, Value=0)
  ]}
]
variables [
 KillCount {type=number}
 KillCounter {type=integer}
 MatInstance {type=MaterialInstanceDynamic}
 ShowVar {type=integer}
 Timer {type=number, default=0.7}
 Timer Handle {type=TimerHandle}
]
